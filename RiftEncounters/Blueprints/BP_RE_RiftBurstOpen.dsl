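; BP_RE_RiftBurstOpen is a Rift Encounters actor that plays a one-shot rift-opening effect at its location.
; On BeginPlay, only the non-authority (Remote) branch runs, so the authority itself plays nothing.
; That branch plays the rift open sound and spawns a Niagara flash raised above the actor with a randomized rotation.
; Overlap and Tick events are empty.
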(event EventBeginPlay
  (bind _self self)
  (bind _returnvalue (Transformation|GetActorLocation _self))
  (Networking|SwitchHasAuthority
    (:Remote
      (Audio|PlaySoundatLocation (Variables|Default|GetRift%20Open%20Sound) _returnvalue)
      (Niagara|SpawnSystematLocation (Variables|Default|GetFX_RiftSpawnIn_Flash) (+ _returnvalue "0.000000,0.000000,150.000000") (Math|Rotator|MakeRotator 0.0 0.0 (Math|Random|RandomFloatinRange -1.8e+02 1.8e+02)) "1.000000,1.000000,1.000000" true true "None" true) ; FlashFX
    )))

(event Collision|EventActorBeginOverlap
  (params (OtherActor :type Actor%20Object%20Reference)))

(event EventTick
  (params (DeltaSeconds :type Float%20%28single-precision%29)))

(fn ConstructionScript)
