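# Sound Cue for a rift encounter opening, mixing close bursts, distant bursts and thunder, each with its own attenuation.
# Both burst layers pick randomly from four waves, with their random nodes excluded from branch culling.
# The thunder wave gets random pitch and a short fixed delay, behind a quality-level node whose last slot is empty.
# The cue plays in the Fort_Gadgets sound class with a concurrency override limiting simultaneous instances.
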SoundCue /RiftEncounters/Sound/RiftEncounter_Open_Cue
root=Mixer_0
nodes [
 {id=Attenuation_0, children [Delay_0], properties {attenuationSettings=&/RiftEncounter_Open_Thunder_Att}}
 {id=Attenuation_1, children [Random_0], properties {attenuationSettings=&/RiftEncounter_Close_Att}}
 {id=Attenuation_3, children [Random_2], properties {attenuationSettings=&/RiftEncounter_Distant_Att}}
 {id=Delay_0, children [Modulator_0], properties {delayMax=0.25, delayMin=0.25}}
 {id=Mixer_0, children [Attenuation_1, Attenuation_3, QualityLevel_0], properties {inputVolume [1,0.5,0.9]}}
 {id=Modulator_0, children [WavePlayer_1], properties {pitchMax=1.2, pitchMin=0.9, volumeMax=1, volumeMin=1}}
 {id=QualityLevel_0, children [Attenuation_0, Attenuation_0, Attenuation_0, null]}
 {
  id=Random_0
  children [WavePlayer_0,WavePlayer_3,WavePlayer_4,WavePlayer_5]
  properties {bShouldExcludeFromBranchCulling=true, weights [1,1,1,1]}
 }
 {
  id=Random_2
  children [WavePlayer_2, WavePlayer_9, WavePlayer_11, WavePlayer_12]
  properties {bShouldExcludeFromBranchCulling=true, weights [1,1,1,1]}
 }
 {id=WavePlayer_0, properties {soundWaveAssetPtr=&/RiftEncounter_Open_Burst_01}}
 {id=WavePlayer_1, properties {soundWaveAssetPtr=&/RiftEncounter_Open_Thunder_01}}
 {id=WavePlayer_11, properties {soundWaveAssetPtr=&/RiftEncounter_Open_Burst_Distant_v2_03}}
 {id=WavePlayer_12, properties {soundWaveAssetPtr=&/RiftEncounter_Open_Burst_Distant_v2_04}}
 {id=WavePlayer_2, properties {soundWaveAssetPtr=&/RiftEncounter_Open_Burst_Distant_v2_01}}
 {id=WavePlayer_3, properties {soundWaveAssetPtr=&/RiftEncounter_Open_Burst_02}}
 {id=WavePlayer_4, properties {soundWaveAssetPtr=&/RiftEncounter_Open_Burst_03}}
 {id=WavePlayer_5, properties {soundWaveAssetPtr=&/RiftEncounter_Open_Burst_04}}
 {id=WavePlayer_9, properties {soundWaveAssetPtr=&/RiftEncounter_Open_Burst_Distant_v2_02}}
]
properties {
 bOverrideConcurrency=true
 maxDistance=25000
 soundClassObject=&Fort_Gadgets
 volumeMultiplier=1.5
 concurrencyOverrides {maxCount=3}
}
