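# Sound cue for the Rift Encounter player push-back effect, routed to the Fort_Gadgets sound class.
# A Quality Level root feeds its first three slots into one shared Random node, leaving the fourth empty.
# The Random node picks with equal weight among three wave variations and is excluded from branch culling.
# A concurrency override allows one instance at a time, stopping the oldest when it retriggers.
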SoundCue /RiftEncounters/Sound/RiftEncounter_Push_Player_Back_Cue
root=QualityLevel_1
nodes [
 {id=QualityLevel_1, children [Random_0,Random_0,Random_0,null]}
 {id=Random_0, children [WavePlayer_0,WavePlayer_3,WavePlayer_4], properties {bShouldExcludeFromBranchCulling=true, weights [1,1,1]}}
 {id=WavePlayer_0, properties {soundWaveAssetPtr=&/RiftEncounter_Push_Player_Back_01}}
 {id=WavePlayer_3, properties {soundWaveAssetPtr=&/RiftEncounter_Push_Player_Back_02}}
 {id=WavePlayer_4, properties {soundWaveAssetPtr=&/RiftEncounter_Push_Player_Back_03}}
]
properties {
 bOverrideConcurrency=true
 maxDistance=2097152
 soundClassObject=&Fort_Gadgets
 volumeMultiplier=1
 concurrencyOverrides {maxCount=1, resolutionRule=StopOldest}
}
