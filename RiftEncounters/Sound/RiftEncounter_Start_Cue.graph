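# Sound cue for the build-up that plays when a Rift Encounter starts, in the Fort_Gadgets sound class.
# A mixer layers a close variant and a quieter distant variant, each through its own attenuation asset.
# Each layer randomly picks from equally weighted waves, with the random nodes excluded from branch culling.
# A concurrency override caps the number of simultaneous instances.
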SoundCue /RiftEncounters/Sound/RiftEncounter_Start_Cue
root=Mixer_0
nodes [
 {id=Attenuation_0, children [Random_0], properties {attenuationSettings=&/RiftEncounter_Close_Att}}
 {id=Attenuation_3, children [Random_3], properties {attenuationSettings=&/RiftEncounter_Distant_Att}}
 {id=Mixer_0, children [Attenuation_0, Attenuation_3], properties {inputVolume [1,0.5]}}
 {
  id=Random_0
  children [WavePlayer_1,WavePlayer_2,WavePlayer_3,WavePlayer_4]
  properties {bShouldExcludeFromBranchCulling=true, weights [1,1,1,1]}
 }
 {id=Random_3, children [WavePlayer_9, WavePlayer_10, WavePlayer_11], properties {bShouldExcludeFromBranchCulling=true, weights [1,1,1]}}
 {id=WavePlayer_1, properties {soundWaveAssetPtr=&/RiftEncounter_Start_Build_Up_01}}
 {id=WavePlayer_10, properties {soundWaveAssetPtr=&/RiftEncounter_Start_Build_Up_Distant_v2_02}}
 {id=WavePlayer_11, properties {soundWaveAssetPtr=&/RiftEncounter_Start_Build_Up_Distant_v2_03}}
 {id=WavePlayer_2, properties {soundWaveAssetPtr=&/RiftEncounter_Start_Build_Up_02}}
 {id=WavePlayer_3, properties {soundWaveAssetPtr=&/RiftEncounter_Start_Build_Up_03}}
 {id=WavePlayer_4, properties {soundWaveAssetPtr=&/RiftEncounter_Start_Build_Up_04}}
 {id=WavePlayer_9, properties {soundWaveAssetPtr=&/RiftEncounter_Start_Build_Up_Distant_v2_01}}
]
properties {
 bOverrideConcurrency=true
 maxDistance=25000
 soundClassObject=&Fort_Gadgets
 volumeMultiplier=0.7
 concurrencyOverrides {maxCount=3}
}
